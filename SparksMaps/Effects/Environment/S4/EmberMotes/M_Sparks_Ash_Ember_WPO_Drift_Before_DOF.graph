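# Unlit translucent Before-DOF particle material for drifting, glowing ash and ember sprites.
# SubUV texture channels split into particle-tinted glow (HDR boosted above low quality), base-color tint and an opacity mask.
# Particle alpha and panned noise modulate that mask, with depth and near-camera fade branches.
# World position offset combines sine-based horizontal drift with height growing over particle life, varied by particle random and dynamic parameters.
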Material /SparksMaps/Effects/Environment/S4/EmberMotes/M_Sparks_Ash_Ember_WPO_Drift_Before_DOF
graph {
 properties {
  bOutputTranslucentVelocity=true
  bUsedWithNiagaraSprites=true
  bUsedWithParticleSprites=true
  blendMode=BLEND_Translucent
  pixelDepthOffsetMode=PDOM_Legacy
  shadingModel=MSM_Unlit
  translucencyPass=MTP_BeforeDOF
  usedShadingModels=Unlit
 }
 nodes [
  n1 = ParticleColor()
  n2 = MaterialFunctionCall(Float3=ParticleSubUV!(UVs, Tex, Apply View MipBias) {texture=/Game/Effects/Fort_Effects/Textures/Motes/T_Ash_Embers_Glow_SubUV}.RGB) {materialFunction=&BreakOutFloat3Components}
  n3 = Multiply(A=@n1.RGB, B=@n2.R)
  n4 = QualitySwitch(
   Default=MaterialFunctionCall(LDR Input=Clamp(None=@n3), HDR Tint=@n3, HDR Intensity Multiplier=ScalarParameter() {defaultValue=2, parameterName=HDRIntensity}) {materialFunction=&DeriveHDRfromLDR}.RemappedRange
   Low=@n3
  )
  n5 = Add(A=@n1.RGB, B=VectorParameter() {parameterName=BaseColor, defaultValue=<rgba 0.0989 0.1471 0.3 0>}.RGB)
  MaterialFunctionCall(Emissive=Add(A=FeatureLevelSwitch(Default=@n4, ES2=@n3, ES3_1=@n3, SM4_REMOVED=@n4), B=Multiply(A=@n5, B=@n2.B))) {materialFunction=&MF_FX_EmissiveAdjustments}
  MaterialFunctionCall() {materialFunction=&DebugTimeSine}
  n6 = Time()
  n7 = ParticleRandom()
  n8 = DynamicParameter()
  n9 = Add(A=@n7, B=@n8.Param1)
  n10 = LinearInterpolate(
   A=ScalarParameter() {defaultValue=0.12, parameterName=WPOSpeedMin}
   B=ScalarParameter() {defaultValue=0.25, parameterName=WPOSpeedMax}
   Alpha=@n9
  )
  n11 = Multiply(A=@n7) {constB=2.3}
  n12 = Add(A=Multiply(A=Multiply(A=@n6, B=@n10), B=@n8.Param4), B=Add(A=Frac(None=@n11), B=@n8.Param2))
  n13 = AppendVector(A=Sine(None=@n12) {period=5.14}, B=Sine(None=@n12) {period=4.12})
  n14 = ScalarParameter() {defaultValue=0.25, parameterName=MinWPOHorizontalDrift}
  n15 = ParticleRelativeTime()
  n16 = Sine(None=Add(A=@n15) {constB=-0.3}) {period=1.5}
  n17 = Multiply(
   A=ScalarParameter() {defaultValue=12, parameterName=WPOHorizontalDrift}
   B=LinearInterpolate(A=Multiply(A=@n8.Param2, B=@n14), Alpha=ConstantBiasScale(None=@n16)) {constA=1}
  )
  n18 = ScalarParameter() {defaultValue=18, parameterName=WPOHeightLimit}
  n19 = Power(Base=@n15, Exp=ScalarParameter() {defaultValue=2, parameterName=WPOHeightLerpPower}) {constExponent=2.4}
  Multiply(A=AppendVector(A=Multiply(A=@n13, B=@n17) {constA=1}, B=LinearInterpolate(B=@n18, Alpha=@n19) {constAlpha=0}), B=@n8.Param3)
  n20 = Multiply(A=@n1.A) {constB=1.2}
  n21 = Multiply(A=Add(A=@n2.R, B=@n2.R)) {constB=3}
  n22 = TextureCoordinate() {uTiling=6, vTiling=6}
  n23 = TextureSample(UVs=Panner(Coordinate=Add(A=@n22, B=@n9)) {speedX=0.4, speedY=0.9}) {texture=&T_Smooth_tile}
  n24 = LinearInterpolate(B=Multiply(A=@n1.A) {constB=10}, Alpha=Multiply(A=@n1.A, B=Add(A=@n21, B=@n23.G))) {constA=-0.2}
  n25 = Multiply(A=@n2.G, B=Clamp(None=Add(A=@n20, B=@n24)))
  n26 = DynamicParameter() {parameterIndex=1, defaultValue=<rgba 10 1 1 1>, paramNames [DepthFade,Param2,Param3,Param4]}
  Multiply(A=@n25, B=QualitySwitch(Default=DepthFade(FadeDistance=@n26.DepthFade) {fadeDistanceDefault=10}, Low=Constant() {r=1}))
  Multiply(
   A=@n25
   B=MaterialFunctionCall(Distance=ScalarParameter() {defaultValue=75, parameterName=Near Camera Fade}) {materialFunction=&Near_Camera_Fade}.Result
  )
  Time() {bOverride_Period=true, period=1}
 ]
}
